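# Balmorhea universal vehicle skin material that draws an animated nebula with bright star sparkles, masked by the BodyTextures green channel.
# Noise textures sampled through rotating, panning and self-distorted UVs form pulsing cloud layers that blend colors.
# Time-driven sine and cosine terms animate pulses and a subtle HueShift, while subtracted noise masks isolate intensely boosted stars.
# BaseColor, AccentColor and PaintColor parameters supply tints, and a Painted static switch alters color blending and star tint.
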Material /VehicleCosmetics/UniversalSkins/Balmorhea/MAT_Balmorhea
graph {
 properties {bUsedWithMeshDeformer=true, bUsedWithSkeletalMesh=true, pixelDepthOffsetMode=PDOM_Legacy, usedShadingModels=Default Lit}
 nodes [
  n1 = NamedRerouteUsage()
  n2 = Fmod(A=Multiply(A=@n1) {constA=3, constB=3}, B=Constant() {r=1})
  n3 = Cosine(None=Multiply(A=@n1) {constB=9}) {period=10}
  n4 = Multiply(A=Saturate(None=Subtract(A=@n2, B=@n3))) {constB=0.5}
  n5 = Abs(None=Cosine(None=@n1) {period=10})
  n6 = ConstantBiasScale(None=Add(A=Multiply(A=@n5) {constB=2}) {constA=0.5, constB=0.5}) {bias=0.3, scale=0.8}
  n7 = Multiply(A=Power(Base=NamedRerouteUsage()) {constExponent=3}, B=ConstantBiasScale(None=NamedRerouteUsage()) {bias=0.8, scale=0.55})
  n8 = Multiply(A=ConstantBiasScale(None=Add(A=@n4, B=@n6)) {bias=0.25, scale=3}, B=Power(Base=@n7) {constExponent=2.75})
  n9 = Multiply(A=NamedRerouteUsage(), B=@n8) {desc=Step Point 01}
  n10 = StaticSwitchParameter(True=Constant() {r=0.01}, False=Constant() {r=0.075}) {desc=Step Point 03, parameterName=Painted}
  n11 = NamedRerouteUsage()
  n12 = Sine(None=Multiply(A=@n11) {constB=0.1})
  n13 = LinearInterpolate(
   A=Multiply(A=NamedRerouteUsage(), B=NamedRerouteUsage())
   B=NamedRerouteUsage()
   Alpha=ConstantBiasScale(None=Multiply(A=@n12) {constB=0.075}) {bias=0.18, scale=1}
  )
  n14 = Power(Base=ConstantBiasScale(None=@n13) {scale=1.25}) {constExponent=0.75}
  n15 = Multiply(A=NamedRerouteUsage(), B=@n14)
  n16 = LinearInterpolate(
   A=@n15
   B=MaterialFunctionCall(Hue Shift Percentage=Multiply(A=@n12, B=@n10), Texture=@n15) {materialFunction=&HueShift}.Result
   Alpha=NamedRerouteUsage()
  )
  n17 = Saturate(None=@n16)
  n18 = NamedRerouteUsage()
  n19 = LinearInterpolate(A=@n17, B=Subtract(A=@n17, B=@n18))
  n20 = NamedRerouteUsage()
  n21 = Saturate(None=Add(A=@n17, B=@n18))
  n22 = StaticSwitchParameter(True=LinearInterpolate(A=Saturate(None=@n19), B=Multiply(A=@n20, B=@n21), Alpha=@n18), False=@n16) {desc=Step Point 03, parameterName=Painted}
  n23 = Saturate(None=Subtract(A=@n14, B=@n8))
  n24 = Saturate(None=Power(Base=LinearInterpolate(A=@n9, B=@n22, Alpha=@n23) {desc=Step Point 02}) {constExponent=1.08})
  n25 = Constant3Vector() {constant=<rgba 1 1 1 0>}
  n26 = Desaturation(None=NamedRerouteUsage()) {luminanceFactors=<rgba 0.3958 1 0 0>}
  n27 = StaticSwitchParameter(True=LinearInterpolate(B=@n25, Alpha=Multiply(A=@n26) {constB=0.5}), False=@n25) {desc=Step Point 03, parameterName=Painted}
  n28 = NamedRerouteUsage()
  n29 = NamedRerouteUsage()
  n30 = Multiply(A=@n28, B=@n29)
  n31 = Add(A=NamedRerouteUsage(), B=@n13) {desc=Step Point 05}
  n32 = Multiply(A=Multiply(A=@n28, B=LinearInterpolate(A=@n30, B=@n29, Alpha=@n28)), B=ConstantBiasScale(None=Saturate(None=@n31)) {bias=0.2, scale=1})
  n33 = OneMinus(None=StaticSwitchParameter(True=NamedRerouteUsage(), False=NamedRerouteUsage()) {desc=Step Point 03, parameterName=Painted}) {desc=Step Point 4}
  n34 = Saturate(None=Subtract(A=Multiply(A=@n32, B=@n33), B=Saturate(None=@n8)) {desc=SubZZZ})
  n35 = Multiply(A=Saturate(None=Power(Base=@n34))) {constB=2000, desc=StarIntensity}
  n36 = TextureSampleParameter2D() {desc=BODYTEXTURE, parameterName=BodyTextures, samplerType=SAMPLERTYPE_LinearColor, texture=/VehicleCosmetics/Textures/T_Black_Linear}
  Multiply(A=Saturate(None=LinearInterpolate(A=@n24, B=@n27, Alpha=@n35)), B=@n36.G)
  NamedRerouteDeclaration(None=TextureCoordinate() {coordinateIndex=1}) {name=UVCoord_01, nodeColor=<rgba 1 0.1981 0 1>}
  NamedRerouteDeclaration(None=TextureObject() {texture=/VehicleCosmetics/Textures/UniversalTextures/Universal_Noise_55_RGB}) {name=NebulaTexture, desc=Uni55, nodeColor=<rgba 1 0 0.1875 1>}
  NamedRerouteDeclaration(None=VectorParameter() {parameterName=PaintColor, sortPriority=9, defaultValue=<rgba 0.459 1 0 1>}.RGB) {name=PaintColor_RGB, nodeColor=<rgba 1 0 0.4219 1>}
  n37 = NamedRerouteUsage()
  n38 = ComponentMask(None=@n37) {r=true}
  n39 = Constant() {r=-0.65}
  n40 = Abs(None=@n39)
  n41 = ComponentMask(None=@n37) {g=true}
  n42 = AppendVector(A=Add(B=Multiply(A=@n38, B=@n40)) {constA=-0.05}, B=Add(B=Multiply(A=@n39, B=@n41) {constB=-0.65}) {constA=0.1})
  n43 = Rotator(Coordinate=@n42) {centerY=0.4, speed=0.3}
  n44 = ConstantBiasScale(None=OneMinus(None=NamedRerouteUsage())) {scale=0.25}
  n45 = Add(A=@n7, B=Add(A=NamedRerouteUsage(), B=@n42))
  n46 = TextureSample(UVs=LinearInterpolate(A=@n42, B=LinearInterpolate(A=@n43, B=@n45, Alpha=@n44)), Tex=NamedRerouteUsage()) {desc=CloudBase}
  n47 = Multiply(A=@n11) {constB=0.05}
  n48 = Constant() {r=0.1}
  n49 = Multiply(A=@n11) {constB=0.1}
  n50 = Add(A=Add(A=Fmod(A=@n47, B=@n48), B=Sine(None=@n49))) {constB=2.75}
  NamedRerouteDeclaration(None=Saturate(None=Power(Base=@n46.G, Exp=@n50))) {name=BaseAndPulse, nodeColor=<rgba 1 0 0.0469 1>}
  n51 = NamedRerouteUsage()
  n52 = Constant4Vector() {constant=<rgba 2.5 2.5 -0.03 0.01>}
  n53 = AppendVector(A=Multiply(A=@n52.R) {constB=7}, B=Divide(A=@n52.G) {constB=3})
  NamedRerouteDeclaration(None=TextureSample(UVs=Multiply(A=@n51, B=@n53), Tex=NamedRerouteUsage()) {desc=Uni 44, samplerSource=SSM_Wrap_WorldGroupSettings}.R) {name=StarNoise_RChan, nodeColor=<rgba 0.2891 0 1 1>}
  NamedRerouteDeclaration(None=TextureObject() {texture=/VehicleCosmetics/Textures/UniversalTextures/Universal_Noise_42_RGB}) {name=Uni44_Texture, desc=Uni44, nodeColor=<rgba 0 0.6484 1 1>}
  n54 = Rotator(Coordinate=NamedRerouteUsage()) {speed=0.03}
  n55 = ConstantBiasScale(None=@n54) {bias=2, scale=1}
  n56 = ConstantBiasScale(None=@n54) {bias=0, scale=0.005}
  NamedRerouteDeclaration(None=TextureSample(UVs=LinearInterpolate(A=@n54, B=@n55, Alpha=@n56), Tex=NamedRerouteUsage()) {desc=BBB, samplerSource=SSM_Wrap_WorldGroupSettings}.G) {name=Nebula_BBB_GChan, nodeColor=<rgba 0 0.8548 1 1>}
  n57 = Panner(Coordinate=NamedRerouteUsage()) {speedY=0.015}
  n58 = ConstantBiasScale(None=@n57) {bias=2, scale=1}
  n59 = ConstantBiasScale(None=NamedRerouteUsage()) {bias=0, scale=0.005}
  NamedRerouteDeclaration(None=TextureSample(UVs=LinearInterpolate(A=@n57, B=@n58, Alpha=@n59), Tex=NamedRerouteUsage()) {desc=CCC, samplerSource=SSM_Wrap_WorldGroupSettings}.G) {name=Nebula_CCC_GChan, nodeColor=<rgba 1 0.7266 0 1>}
  n60 = Rotator(Coordinate=NamedRerouteUsage()) {centerY=0.55, speed=-0.2}
  n61 = ConstantBiasScale(None=@n60)
  n62 = ConstantBiasScale(None=NamedRerouteUsage()) {bias=0.1, scale=0.15}
  NamedRerouteDeclaration(None=TextureSample(UVs=LinearInterpolate(A=@n60, B=@n61, Alpha=@n62), Tex=NamedRerouteUsage()) {desc=DDD, samplerSource=SSM_Wrap_WorldGroupSettings}.B) {name=Nebula_DDD_BChan, nodeColor=<rgba 0.0337 0 1 1>}
  NamedRerouteDeclaration(None=TextureSample(UVs=NamedRerouteUsage(), Tex=NamedRerouteUsage()) {desc=AAA, samplerSource=SSM_Wrap_WorldGroupSettings}.G) {name=Nebula_AAA_GChan, nodeColor=<rgba 0.0078 1 0 1>}
  NamedRerouteDeclaration(None=VectorParameter() {parameterName=AccentColor, sortPriority=7, defaultValue=<rgba 0.895 0 1 1>}.RGB) {name=CustomColor_RGB, nodeColor=<rgba 1 0 0.2812 1>}
  NamedRerouteDeclaration(None=Time()) {name=Time_NR, nodeColor=<rgba 0 0.7891 1 1>}
  n63 = Multiply(A=NamedRerouteUsage(), B=AppendVector(A=@n52.R, B=@n52.R))
  n64 = Multiply(A=NamedRerouteUsage(), B=AppendVector(A=@n52.B, B=@n52.B))
  NamedRerouteDeclaration(None=TextureSample(UVs=Add(A=@n63, B=@n64), Tex=NamedRerouteUsage()) {desc=JJJ, samplerSource=SSM_Wrap_WorldGroupSettings}.R) {name=JJJ_Nubula_RChan, nodeColor=<rgba 0.4062 1 0 1>}
  NamedRerouteDeclaration(None=VectorParameter() {parameterName=BaseColor, sortPriority=5, defaultValue=<rgba 0 0.6576 0.8176 1>}.RGB) {name=TeamColor_RGB, nodeColor=<rgba 1 0 0.1875 1>}
 ]
}
